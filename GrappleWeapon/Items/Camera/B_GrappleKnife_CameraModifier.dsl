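; Camera modifier for the grapple knife that adds a dynamic chase-camera offset and widening FOV while the player zips.
; It caches the pawn, knife and launch point, then derives pull-back from distance travelled.
; Vertical and sideways offsets follow velocity, smoothed by interpolation and scaled by a zoom-out coefficient.
; Offsets blend by modifier alpha; after the scaled zip duration, DisableAndRemove fades out and removes the modifier.
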(event Custom|DisableAndRemove
  (Variables|Default|SetUseEndAlpha true)
  (CameraModifier|DisableModifier false)
  (Utilities|FlowControl|Delay (Variables|CameraModifier|GetAlphaOutTime))
  (Variables|Default|SetRemoved true)
  (Variables|Default|SetWeapon 0)
  (bind _cameraowner (Variables|CameraModifier|GetCameraOwner)
    (:then
      (Utilities|IsValid (Game|Player|GetOwningPlayerController _cameraowner)
        (:Is%20Valid
          (Game|Player|RemoveCameraModifier (Variables|/Script/Engine.PlayerController|GetPlayerCameraManager (Game|Player|GetOwningPlayerController _cameraowner)) self))))
    (:else)))

(fn BlueprintModifyCamera
  (params (DeltaTime :type Float%20%28single-precision%29) (ViewLocation :type Vector :default "0, 0, 0") (ViewRotation :type Rotator :default "0, 0, 0") (FOV :type Float%20%28single-precision%29))
  (bind (_returnvalue_x _returnvalue_y _returnvalue_z) (Math|Vector|BreakVector (Transformation|GetVelocity (Variables|Default|GetPlayer))))
  (Utilities|FlowControl|Sequence
    (:then_0
      (Utilities|IsValid (Variables|Default|GetWeapon)
        (:Is%20Not%20Valid
          (bind _asfort_player_controller (Utilities|Casting|CastToFortPlayerController (Game|Player|GetOwningPlayerController (Variables|CameraModifier|GetCameraOwner))))
          (bind _output_get (Variables|Default|SetPlayer (Game|Player|GetPlayerPawn _asfort_player_controller)))
          (Variables|Default|SetLaunchedPosition (Transformation|GetActorLocation _output_get))
          (bind _asb_grapple_knife (Utilities|Casting|CastToB_GrappleKnife (Weapon|GetCurrentWeapon (Variables|Default|GetPlayer))))
          (Variables|Default|SetWeapon _asb_grapple_knife)
          (Variables|Default|SetStartedTime (Utilities|Time|GetGameTimeinSeconds)))))
    (:then_1
      (if (and (Utilities|IsValid (Variables|Default|GetWeapon) :!class /Script/Engine.KismetSystemLibrary) (Utilities|IsValid (Variables|Default|GetPlayer) :!class /Script/Engine.KismetSystemLibrary))
        (if (Variables|Default|GetRemoved)
          (CallFunction|DisableandRemove)))
      (return (Variables|GetViewLocation) (Variables|GetViewRotation) (Variables|GetFOV)))
    (:then_2
      (Variables|Default|SetFOV%20Offset (+ (Variables|Default|GetFOV%20Offset) (* (Variables|GetDeltaTime) 5e+01))))
    (:then_3
      (bind _output_get_2 (Variables|Default|SetCurrent%20Forward (- (* (Math|Vector|VectorLength (- (Variables|Default|GetLaunchedPosition) (Transformation|GetActorLocation (Variables|Default|GetPlayer)))) (Utilities|RemapFloat (Math|Vector|VectorLength (- (Variables|Default|GetLaunchedPosition) (Transformation|GetActorLocation (Variables|Default|GetPlayer)))) "(X=1000.000000,Y=3000.000000)" "(X=-0.500000,Y=-0.300000)" true)) 1.5e+03)))
      (Variables|Default|SetZoomOutCoefficient (Utilities|RemapFloat (+ (* (Variables|Default|GetFOVOffset) 5e+01) (* (Variables|Default|GetCurrent%20Forward) -1.0)) "(X=0.000000,Y=1500.000000)" "(X=0.300000,Y=1.500000)" true))
      (Variables|Default|SetCurrent%20Up (Math|Interpolation|FInterpTo (Variables|Default|GetCurrent%20Up) (* (Utilities|RemapFloat _returnvalue_z "(X=-4000.000000,Y=2500.000000)" "(X=500.000000,Y=-300.000000)" true) (Variables|Default|GetZoomOutCoefficient)) (Variables|GetDeltaTime) 2.0))
      (Variables|Default|SetCurrent%20Right (Math|Interpolation|FInterpTo (Variables|Default|GetCurrent%20Right) (* (Utilities|RemapFloat (Math|Vector|DotProduct (Transformation|GetVelocity (Variables|Default|GetPlayer)) (Math|Vector|GetRightVector (Variables|GetViewRotation))) "(X=-700.000000,Y=700.000000)" "(X=240.000000,Y=-240.000000)" true) (Variables|Default|GetZoomOutCoefficient)) (Variables|GetDeltaTime) 2.0)))
    (:then_4
      (if (>= (- (Utilities|Time|GetGameTimeinSeconds) (Variables|Default|GetStartedTime)) (* (Variables|/GrappleWeapon/Items/GrappleKnife/V2/B_GrappleKnife.B_GrappleKnife_C|GetZipDuration (Variables|Default|GetWeapon)) (Variables|Default|GetZipDurationMulti)))
        (CallFunction|DisableandRemove)))
    (:then_5
      (return (+ (Variables|GetViewLocation) (* (Utilities|Operators|Add (* (Math|Vector|GetForwardVector (Variables|GetViewRotation)) (Variables|Default|GetCurrent%20Forward)) (* (Math|Vector|GetRightVector (Variables|GetViewRotation)) (Variables|Default|GetCurrent%20Right)) (Math|Vector|MakeVector 0.0 0.0 (Variables|Default|GetCurrent%20Up))) (Variables|CameraModifier|GetAlpha))) (Variables|GetViewRotation) (+ (Variables|GetFOV) (* (Variables|Default|GetFOV%20Offset) (Variables|CameraModifier|GetAlpha)))))))
